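# Physics asset for the Frogman backpack cosmetic, providing collision bodies and joints for the pack and its gun and knife attachments.
# The pack bone carries two overlapping angled boxes, while dyn_gun and dyn_knife each use one slim box.
# Each attachment is constrained directly to the pack with limited swing and very tight twist, the gun allowing wider swing but less twist than the knife.
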PhysicsAsset /BRCosmetics/Accessories/FORT_Backpacks/Backpack_F_MED_Frogman/Meshes/F_MED_Frogman_Physics
bodies {
 dyn_gun [{shapeType=Box, extentX=2.3973, extentY=4, extentZ=20.3172, center=<0 0 1>, rotation=<rot 90 0 0>}]
 dyn_knife [{shapeType=Box, extentX=1, extentY=4, extentZ=20, center=<0 0 1>, rotation=<rot 90 0 0>}]
 pack [
  {shapeType=Box, extentX=16.8096, extentY=10.0383, extentZ=12.0647, center=<1.2276 -2.5187 -8.4574>, rotation=<rot 71.3004 0 0>}
  {shapeType=Box, extentX=16.8096, extentY=10.0383, extentZ=12.0647, center=<-3.236 -2.5187 -6.666>, rotation=<rot 72.8162 180 180>}
 ]
}
constraints [
 {bone1Name=dyn_knife, bone2Name=pack, swing1LimitDegrees=33, swing2LimitDegrees=33, twistLimitDegrees=6}
 {bone1Name=dyn_gun, bone2Name=pack, swing1LimitDegrees=45, swing2LimitDegrees=45, twistLimitDegrees=2}
]
